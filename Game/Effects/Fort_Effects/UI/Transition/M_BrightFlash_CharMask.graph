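# M_BrightFlash_CharMask is a translucent unlit sprite material for a bright flash transition effect, with a stencil-based mask, likely for characters.
# Particle RGB drives emissive through an adjustment function, while particle alpha dissolves a tiled teleporter grid texture.
# The grid multiplies with a horizontal center-peaked falloff and an inverted depth fade, concentrating it near intersecting geometry.
# An inverted StencilMaskCompare result cuts matching stencil pixels out of that combined mask.
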Material /Game/Effects/Fort_Effects/UI/Transition/M_BrightFlash_CharMask
graph {
 properties {
  bEnableResponsiveAA=true
  bUsedWithNiagaraSprites=true
  bUsedWithParticleSprites=true
  blendMode=BLEND_Translucent
  pixelDepthOffsetMode=PDOM_Legacy
  shadingModel=MSM_Unlit
  usedShadingModels=Unlit
 }
 nodes [
  n1 = ParticleColor()
  MaterialFunctionCall(Emissive=@n1.RGB) {materialFunction=&MF_FX_EmissiveAdjustments}
  n2 = Subtract(A=ComponentMask(None=TextureCoordinate()) {r=true}) {constB=0.5}
  n3 = OneMinus(None=Multiply(A=Abs(None=@n2)) {constB=2})
  n4 = ScalarParameter() {defaultValue=5, parameterName=UTile}
  n5 = ScalarParameter() {defaultValue=5, parameterName=VTile}
  n6 = TextureSample(UVs=Multiply(
    A=TextureCoordinate()
    B=MaterialFunctionCall(X=@n4, Y=@n5) {materialFunction=&MakeFloat2}.Result
   )) {samplerType=SAMPLERTYPE_LinearColor, texture=/Game/Gadgets/Assets/Teleporter/Textures/T_Teleporter_GridSquares_01}
  n7 = Saturate(None=Add(A=OneMinus(None=@n6.R), B=LinearInterpolate(Alpha=@n1.A) {constA=-1}))
  n8 = OneMinus(None=DepthFade(FadeDistance=Constant() {r=350}))
  n9 = MaterialFunctionCall!(ID=Constant()) {materialFunction=/Game/Packages/Fortress_SharedMaterials/Stencil_ID_Definitions/StencilMaskCompare}
  Multiply(A=Clamp(None=Power(Base=@n3) {constExponent=2.5}), B=Multiply(A=Multiply(A=@n7, B=@n8), B=OneMinus(None=@n9.Result)))
  QualitySwitch()
  SceneTexture() {sceneTextureId=PPI_WorldNormal}
 ]
}
