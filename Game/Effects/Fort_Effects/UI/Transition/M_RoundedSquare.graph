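# Unlit translucent particle sprite material that draws a square with rounded corners, filed under UI transition effects.
# Particle color RGB passes through an emissive adjustment function.
# Opacity multiplies particle alpha by a UV-centered sphere mask, whose radius reaches past the edges but clips the corners, avoiding a texture.
# Depth fade, bypassed on mobile, and a parameterized camera depth fade further scale that opacity.
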Material /Game/Effects/Fort_Effects/UI/Transition/M_RoundedSquare
graph {
 properties {
  bUsedWithNanite=true
  bUsedWithNiagaraSprites=true
  bUsedWithParticleSprites=true
  blendMode=BLEND_Translucent
  pixelDepthOffsetMode=PDOM_Legacy
  shadingModel=MSM_Unlit
  usedShadingModels=Unlit
 }
 nodes [
  n1 = ParticleColor()
  MaterialFunctionCall(Emissive=@n1.RGB) {materialFunction=&MF_FX_EmissiveAdjustments}
  n2 = SphereMask(A=TextureCoordinate(), B=Constant() {r=0.5}, Radius=Constant() {r=0.65}, Hardness=Constant() {r=1})
  n3 = DepthFade(FadeDistance=Constant() {r=10})
  n4 = Constant() {r=1}
  Multiply(
   A=Multiply(A=Multiply(A=@n1.A, B=@n2), B=ShadingPathSwitch(Default=@n3, Mobile=@n4) {desc=Migrated from FeatureLevelSwitch})
   B=MaterialFunctionCall(Fade Length=Constant() {r=100}, Fade Offset=ScalarParameter() {defaultValue=-96000, parameterName=CameraDepthFade}) {materialFunction=&CameraDepthFade}.Result
  )
 ]
}
