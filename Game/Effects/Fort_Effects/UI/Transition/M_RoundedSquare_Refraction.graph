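# Additive unlit particle material for a UI transition effect that adds refraction distortion in a rounded-square shape.
# A hard sphere mask on UVs, wider than the quad's half-width, forms the shape; particle alpha and depth fade scale it to drive the Refraction parameter blend.
# Particle color feeds an emissive adjustment function, and the Random dynamic parameter blends two constant vectors.
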Material /Game/Effects/Fort_Effects/UI/Transition/M_RoundedSquare_Refraction
graph {
 properties {
  bUsedWithNanite=true
  bUsedWithNiagaraSprites=true
  bUsedWithParticleSprites=true
  blendMode=BLEND_Additive
  pixelDepthOffsetMode=PDOM_Legacy
  refractionMethod=RM_IndexOfRefraction
  shadingModel=MSM_Unlit
  usedShadingModels=Unlit
 }
 nodes [
  n1 = ParticleColor()
  n2 = SphereMask(A=TextureCoordinate(), B=Constant() {r=0.5}, Radius=Constant() {r=0.68}, Hardness=Constant() {r=1})
  LinearInterpolate(
   B=ScalarParameter() {defaultValue=1.1, parameterName=Refraction}
   Alpha=Multiply(A=Multiply(A=@n1.A, B=@n2), B=DepthFade() {fadeDistanceDefault=25})
  ) {constA=1, constB=1.1}
  LinearInterpolate(
   A=Constant3Vector() {constant=<rgba 1 0 0.5 0>}
   B=Constant3Vector() {constant=<rgba 0 1 0.5 0>}
   Alpha=DynamicParameter() {paramNames [Random,Param2,Param3,Param4]}.Random
  )
  MaterialFunctionCall(Emissive=@n1.RGB) {materialFunction=&MF_FX_EmissiveAdjustments}
 ]
}
