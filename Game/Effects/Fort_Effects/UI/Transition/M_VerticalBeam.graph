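# Translucent unlit material for mesh particles that draws scrolling vertical beam streaks, stored with UI transition effects.
# A vertically stretched, panning noise texture passes through Frac and Floor to form hard-edged streaks, with Dynamic Parameters driving thickness and scroll speed.
# The streak mask is multiplied by power-curve fades at both V ends, a Fresnel term, particle alpha and depth fade.
# Particle color feeds a shared emissive-adjustment function.
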Material /Game/Effects/Fort_Effects/UI/Transition/M_VerticalBeam
graph {
 properties {
  bUsedWithMeshParticles=true
  bUsedWithNanite=true
  bUsedWithNiagaraMeshParticles=true
  blendMode=BLEND_Translucent
  pixelDepthOffsetMode=PDOM_Legacy
  shadingModel=MSM_Unlit
  usedShadingModels=Unlit
 }
 nodes [
  n1 = ParticleColor()
  MaterialFunctionCall(Emissive=@n1.RGB) {materialFunction=&MF_FX_EmissiveAdjustments}
  n2 = MaterialFunctionCall(Exponent=Constant() {r=4}) {materialFunction=&MF_Fresnel_FX}
  n3 = DynamicParameter() {defaultValue=<rgba 0.5 -1 1 1>, paramNames [Thickness,ScrollSpeed,Param3,Param4]}
  n4 = TextureSample(UVs=Panner(
    Coordinate=TextureCoordinate() {uTiling=7, vTiling=0.1}
    Speed=MaterialFunctionCall(Y=@n3.ScrollSpeed) {materialFunction=&MakeFloat2}.Result
   ) {speedY=-1}) {texture=&T_NoisePack_01}
  n5 = Floor(None=Add(A=Frac(None=@n4.R), B=DynamicParameter() {defaultValue=<rgba 0.5 -1 1 1>, paramNames [Thickness,ScrollSpeed,Param3,Param4]}.Thickness) {constB=0.5})
  n6 = OneMinus(None=ComponentMask(None=TextureCoordinate()) {g=true})
  n7 = ComponentMask(None=TextureCoordinate()) {g=true}
  n8 = Multiply(A=OneMinus(None=Power(Base=@n6) {constExponent=3}), B=OneMinus(None=Power(Base=@n7) {constExponent=3}))
  Multiply(A=Multiply(A=Multiply(A=@n1.A, B=@n2.Fresnel), B=Multiply(A=@n5, B=@n8)), B=DepthFade(FadeDistance=Constant() {r=12}))
 ]
}
